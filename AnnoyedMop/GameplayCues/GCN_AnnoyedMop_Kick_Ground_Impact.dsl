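; Burst gameplay cue for the Annoyed Mop's kick ground impact, playing effects at the target actor's location.
; After calling the parent burst handler, it spawns a radial force physics field actor only when visual effect quality is 3 or higher.
; It then plays the tagged AOE ground impact sound through the SoundLibrary subsystem and shows a local gunshot-type sound indicator using the impact sound cue.
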(fn OnBurst(EitherVFX-preferthisversion)
  (params (MyTarget :type Actor%20Object%20Reference) (Parameters :type Gameplay%20Cue%20Parameters%20Structure%20%28by%20ref%29) (ParticleComponents :type Array%20of%20FXSystem%20Component%20Object%20References) (AudioComponents :type Array%20of%20Audio%20Component%20Object%20References) (BurstCameraShakeInstance :type Camera%20Shake%20Base%20Object%20Reference) (BurstDecalInstance :type Decal%20Actor%20Object%20Reference))
  (bind _returnvalue (Transformation|GetActorLocation MyTarget))
  (|CallParentFunction MyTarget Parameters ParticleComponents AudioComponents BurstCameraShakeInstance BurstDecalInstance)
  (Utilities|FlowControl|Sequence
    (:then_0
      (bind _returnvalue_2 (Settings|GetGameUserSettings))
      (if (>= (Settings|GetVisualEffectQuality _returnvalue_2) 3)
        (Game|SpawnActorfromClass MyTarget "/Game/Effects/Niagara/Blueprints/PhsyicsFields/FN_RadialForce.FN_RadialForce_C" (Math|Transform|MakeTransform _returnvalue) "AlwaysSpawn" "MultiplyWithRoot" 0 0.05 0.6 2.5e+03 1.2 0.0 0)))
    (:then_1
      (SoundLibrary|PlaySoundSimple (WorldSubsystems|GetSoundLibrarySubsystem MyTarget) MyTarget (Utilities|Struct|MakeSimpleSoundSettings "(TagName=\"Soundlibrary.AnnoyedMop.AOE.GroundImpact\")" false false true false 1.0 1.0 0.0 0 0 "" 0 "KeepRelativeOffset" _returnvalue) :!class /Script/SoundLibrary.SoundLibrarySubsystem)
      (Fortnite|ShowSoundIndicatorLocal :TrackedActor (Variables|GetMyTarget) :Location _returnvalue :MaxAudibleDistance 0.0 :IndicatorType "Gunshot" :Tint "(R=0.000000,G=0.000000,B=0.000000,A=0.000000)" :SoundBase "/AnnoyedMop/Assets/Audio/AnnoyedMop_AOE_Ground_Impact_Cue.AnnoyedMop_AOE_Ground_Impact_Cue" :AttenuationMaxRangeMultiplier 1.0 :AttenuationIntensityMultiplier 1.0)
      (return))))
